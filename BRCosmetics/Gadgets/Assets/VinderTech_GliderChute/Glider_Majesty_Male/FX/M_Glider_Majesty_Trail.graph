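# Translucent unlit Niagara ribbon material for the trail of the Majesty (male) glider.
# The default look layers an edge line and curve-tinted wave and smoke patterns warped by two panning noises, with a dynamic Dark blend.
# Scalar lerps swap in a glittering sand look or a panning crystal-only look.
# Ribbon masks shape opacity, which uses particle alpha, near-camera fade, and depth fade on higher-quality deferred paths.
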Material /BRCosmetics/Gadgets/Assets/VinderTech_GliderChute/Glider_Majesty_Male/FX/M_Glider_Majesty_Trail
graph {
 properties {
  bUsedWithNiagaraRibbons=true
  blendMode=BLEND_Translucent
  pixelDepthOffsetMode=PDOM_Legacy
  shadingModel=MSM_Unlit
  twoSided=true
  usedShadingModels=Unlit
 }
 nodes [
  n1 = DynamicParameter() {defaultValue=<rgba 1 0 0.5 1>, paramNames [LineEmissive, Dark, SandIntensity, Param4]}
  n2 = VectorParameter() {parameterName=LineColor, defaultValue=<rgba 1 1 1 0>}
  n3 = TextureCoordinate()
  n4 = VectorParameter() {parameterName=Line, defaultValue=<rgba 50 10 0 0>}
  n5 = Multiply(A=Power(Base=ComponentMask(None=@n3) {g=true}, Exp=@n4.R), B=@n4.G)
  n6 = CurveAtlasRowParameter(CurveTime=ComponentMask(None=TextureCoordinate()) {g=true}) {
   atlas=/Game/Athena/LinearColorCurves/VFX_CurveAtlas
   curve=/Game/Gadgets/Assets/VinderTech_GliderChute/Glider_Majesty_Male/FX/Curve_Majesty_Ramp
   defaultValue=25
   parameterName=Curve
  }
  n7 = VectorParameter() {parameterName=Tile UV, defaultValue=<rgba 2 2 -3 0>}
  n8 = Multiply(A=TextureCoordinate(), B=AppendVector(A=@n7.R, B=@n7.R))
  n9 = Multiply(A=AppendVector(A=@n7.B, B=@n7.B), B=Time())
  n10 = TextureCoordinate()
  n11 = VectorParameter() {parameterName=NoiseA UV, defaultValue=<rgba 1 1 -0.5 0>}
  n12 = AppendVector(A=@n11.R, B=@n11.R)
  n13 = AppendVector(A=@n11.B, B=@n11.B)
  n14 = Time()
  n15 = TextureSample(UVs=Add(A=Multiply(A=@n10, B=@n12), B=Multiply(A=@n13, B=@n14))) {samplerType=SAMPLERTYPE_Grayscale, texture=/Game/Global/Textures/Default/Noise/TilingNoise06}
  n16 = VectorParameter() {parameterName=Noise, defaultValue=<rgba -0.3 0.3 0.1 -0.1>}
  n17 = TextureCoordinate()
  n18 = VectorParameter() {parameterName=NoiseB UV, defaultValue=<rgba 1 1 -0.3 0>}
  n19 = AppendVector(A=@n18.R, B=@n18.R)
  n20 = AppendVector(A=@n18.B, B=@n18.B)
  n21 = Time()
  n22 = TextureSample(UVs=Add(A=Multiply(A=@n17, B=@n19), B=Multiply(A=@n20, B=@n21))) {samplerType=SAMPLERTYPE_Grayscale, texture=/BRCosmetics/Weapons/FORT_Melee/Pickaxe_Cube_Ninja_Male/FX/T_CellNoise}
  n23 = Multiply(A=Add(A=Multiply(A=@n15.R, B=@n16.R), B=@n16.B), B=Add(A=Multiply(A=@n22.R, B=@n16.G), B=@n16.A))
  n24 = TextureSample(UVs=Add(A=Add(A=@n8, B=@n9), B=@n23)) {texture=/Game/Global/Textures/Default/Tileable_Patterns/T_WavePattern_01}
  n25 = VectorParameter() {parameterName=Tile Emissive Refine, defaultValue=<rgba 5 50 10 0.5>}
  n26 = LinearInterpolate(A=Multiply(A=Power(Base=@n24.R, Exp=@n25.R), B=@n25.G), B=Multiply(A=@n24.R, B=@n25.B), Alpha=@n25.A)
  n27 = ParticleColor()
  n28 = Add(A=Multiply(A=Multiply(A=@n1.LineEmissive, B=@n2.RGB), B=@n5), B=Multiply(A=Multiply(A=@n6, B=@n26), B=@n27.RGB))
  n29 = VectorParameter() {parameterName=Tile2 UV, defaultValue=<rgba 2 2 -3 0>}
  n30 = Multiply(A=TextureCoordinate(), B=AppendVector(A=@n29.R, B=@n29.R))
  n31 = Multiply(A=AppendVector(A=@n29.B, B=@n29.B), B=Time())
  n32 = OneMinus(None=@n23)
  n33 = TextureSample(UVs=Add(A=Add(A=@n30, B=@n31), B=Multiply(A=@n32) {constB=0.5})) {texture=&T_Inky_Smoke_Tile}
  n34 = VectorParameter() {parameterName=Tile Alpha Refine, defaultValue=<rgba 5 50 10 0.5>}
  n35 = LinearInterpolate(A=Multiply(A=Power(Base=@n24.R, Exp=@n34.R), B=@n34.G), B=Multiply(A=@n24.R, B=@n34.B), Alpha=@n34.A)
  n36 = Multiply(A=@n23) {constB=3}
  n37 = TextureCoordinate()
  n38 = OneMinus(None=ComponentMask(None=Add(A=@n36, B=@n37)) {r=true})
  n39 = Power(Base=@n38) {constExponent=20}
  n40 = Multiply(A=Multiply(A=Power(Base=@n38) {constExponent=5}, B=OneMinus(None=@n39))) {constB=2}
  n41 = TextureCoordinate()
  n42 = MaterialFunctionCall(Float2=@n41) {materialFunction=&BreakOutFloat2Components}
  n43 = Power(Base=@n42.G) {constExponent=3}
  n44 = Add(A=Saturate(None=@n5), B=Add(A=Add(A=@n33.R, B=@n35), B=Multiply(A=@n40, B=@n43) {constB=2}))
  n45 = Constant() {r=1}
  n46 = LinearInterpolate(
   A=@n28
   B=Add(A=Multiply(A=@n27.RGB, B=@n44), B=Desaturation(None=@n28, Fraction=@n45))
   Alpha=DynamicParameter() {defaultValue=<rgba 1 0 0.5 1>, paramNames [LineEmissive, Dark, SandIntensity, Param4]}.Dark
  )
  n47 = Constant() {r=24}
  n48 = TextureCoordinate() {coordinateIndex=1, uTiling=8, vTiling=8}
  n49 = TextureSample(UVs=Add(
    A=MaterialFunctionCall!(TextureResolution=@n47) {materialFunction=/Engine/Functions/Engine_MaterialFunctions02/Texturing/ScreenAlignedPixelToPixelUVs}.Result
    B=Multiply(A=@n48) {constB=0.2}
   )) {samplerType=SAMPLERTYPE_Normal, texture=/Game/Environments/World/Backgrounds/Locales/Temperate/Textures/T_WaterNoise}
  n50 = Multiply(A=ConstantBiasScale(None=@n49.RGB), B=Time())
  n51 = CurveAtlasRowParameter(CurveTime=MaterialFunctionCall(Float3=Frac(None=@n50)) {materialFunction=&BreakOutFloat3Components}.R) {
   atlas=&AthenaCurveAtlas01
   curve=/Game/Athena/LinearColorCurves/Curve_Sand_EmissiveColor
   defaultValue=43
   group=Sand
   parameterName=Emissive
  }
  n52 = Multiply(A=@n51) {constB=0.1}
  n53 = Constant() {r=1}
  n54 = Multiply(
   A=Multiply(A=@n27.RGB, B=Desaturation(None=@n52, Fraction=@n53))
   B=ScalarParameter() {defaultValue=1, group=Sand, parameterName=Sand Emissive}
  )
  n55 = ScalarParameter() {group=Sand, parameterName=Use Sand}
  n56 = TextureSample(UVs=Panner(Coordinate=TextureCoordinate() {uTiling=2, vTiling=2}) {speedX=-3}) {texture=/BRCosmetics/Weapons/FORT_Melee/Pickaxe_Majesty_Male/FX/T_Crystal}
  n57 = Add(A=Multiply(A=Power(Base=@n56.R) {constExponent=7}) {constB=50}, B=@n56.R)
  n58 = ScalarParameter() {parameterName=Only Crystal Trail}
  MaterialFunctionCall(Emissive=LinearInterpolate(A=LinearInterpolate(A=@n46, B=@n54, Alpha=@n55), B=Multiply(A=@n27.RGB, B=@n57), Alpha=@n58)) {materialFunction=&MF_FX_EmissiveAdjustments}
  n59 = Multiply(A=OneMinus(None=@n42.R), B=@n43)
  n60 = MaterialFunctionCall(Float2=@n41) {materialFunction=&BreakOutFloat2Components}
  n61 = TextureSample(UVs=AppendVector(A=OneMinus(None=@n60.R), B=@n60.G)) {texture=/Game/Effects/Fort_Effects/Textures/Ramps/T_Ribbon_Ramp}
  n62 = LinearInterpolate(A=Saturate(None=Multiply(A=@n61.R) {constB=2}), B=@n61.R, Alpha=ScalarParameter() {parameterName=Only Crystal Trail})
  n63 = Saturate(None=Multiply(A=Multiply(A=@n59, B=@n62), B=@n42.G))
  n64 = Multiply(A=@n44, B=@n63)
  n65 = DynamicParameter() {parameterIndex=1, defaultValue=<rgba 1 0 0 1>, paramNames [R,G,B,Param4]}
  n66 = DotProduct(
   A=@n49.RGB
   B=MaterialFunctionCall(R=@n65.R, G=@n65.R, B=@n65.R, A=@n65.R) {materialFunction=&AppendMany}.RGB
  )
  n67 = VectorParameter() {group=Sand, parameterName=Default UV, channelNames {a="", b="", g=V_Scale, r=U_Scale}, defaultValue=<rgba 0.5 1 -1 0>}
  n68 = Multiply(A=TextureCoordinate() {coordinateIndex=1}, B=AppendVector(A=@n67.U_Scale, B=@n67.U_Scale))
  n69 = Multiply(A=AppendVector(A=@n67.B, B=@n67.B), B=Time())
  n70 = TextureCoordinate() {coordinateIndex=1}
  n71 = VectorParameter() {group=Sand, parameterName=Sadn Noise UV, channelNames {a="", b="", g=V_Scale, r=U_Scale}, defaultValue=<rgba 0.5 0.5 -0.3 0>}
  n72 = AppendVector(A=@n71.U_Scale, B=@n71.U_Scale)
  n73 = AppendVector(A=@n71.B, B=@n71.B)
  n74 = Time()
  n75 = TextureSample(UVs=Add(A=Multiply(A=@n70, B=@n72), B=Multiply(A=@n73, B=@n74))) {texture=&T_Inky_Smoke_Tile}
  n76 = ScalarParameter() {defaultValue=0.2, group=Sand, parameterName=Sand Noise Intensity}
  n77 = TextureSampleParameter2D(UVs=Add(A=Add(A=@n68, B=@n69), B=Multiply(A=@n75.R, B=@n76) {constB=0.15})) {group=Sand, parameterName=Default Tex, texture=/Game/Global/Textures/Default/Noise/TankMoteNoise}
  n78 = MaterialFunctionCall(Float2=TextureCoordinate()) {materialFunction=&BreakOutFloat2Components}
  n79 = MaterialFunctionCall(X=@n78.G, Y=@n78.G) {materialFunction=&MakeFloat2}
  n80 = Multiply(
   A=Power(Base=Multiply(A=@n77.R, B=@n77.R), Exp=ScalarParameter() {defaultValue=2, group=Sand, parameterName=Default Power}) {constExponent=1}
   B=Multiply(
    A=@n63
    B=TextureSample(UVs=@n79.Result) {samplerSource=SSM_Clamp_WorldGroupSettings, texture=/Game/Effects/Fort_Effects/Textures/Ramps/T_Ramp_Smooth_Reverse}.R
   )
  )
  n81 = DynamicParameter() {defaultValue=<rgba 1 0 0.5 1>, paramNames [LineEmissive, Dark, SandIntensity, Param4]}
  n82 = Multiply(A=Ceil(None=Saturate(None=@n66)), B=Saturate(None=Multiply(A=@n80, B=@n81.SandIntensity)))
  n83 = Multiply(A=@n27.A, B=LinearInterpolate(A=LinearInterpolate(A=@n64, B=@n82, Alpha=@n55), B=Multiply(A=@n57, B=@n63), Alpha=@n58))
  n84 = DepthFade(Opacity=@n83) {fadeDistanceDefault=20}
  n85 = QualitySwitch(Default=@n84, Low=@n83, High=@n84, Medium=@n83, Epic=@n84)
  Multiply(
   A=ShadingPathSwitch(Default=@n85, Deferred=@n85, Forward=@n83, Mobile=@n83)
   B=MaterialFunctionCall(Distance=Constant() {r=128}) {materialFunction=&Near_Camera_Fade}.Result
  )
  ScalarParameter() {defaultValue=0.05, group=Sand, parameterName=Sand Intensity}
 ]
}
