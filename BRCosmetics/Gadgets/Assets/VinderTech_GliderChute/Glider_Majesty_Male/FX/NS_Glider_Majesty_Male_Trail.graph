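# Niagara trail effect for the Majesty Male glider chute, shown during full deployment, with separate tuning for front-end display.
# Four ribbon emitters (smoke, crystal, two sand) stretch between two socket pairs, while two mesh emitters shed dark and gold crystal shards.
# User.Type_Transformation thresholds at 8 and 16 select color, brightness, dark-to-gold shard mix and sand visibility.
# In gameplay, an integrator ramps spawn rates and ribbon alpha.
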NiagaraSystem /BRCosmetics/Gadgets/Assets/VinderTech_GliderChute/Glider_Majesty_Male/FX/NS_Glider_Majesty_Male_Trail
dependencies {
 usedDataInterfaces [&NiagaraDataInterfaceColorCurve, &NiagaraDataInterfaceCurve, &NiagaraDataInterfaceArrayInt32]
 usedDynamicInputs []
 usedModules [&SystemState, &FN_Integrator]
}
emitters {
 Crystal_Piece_Dark {
  emitterUpdateScript [
   &EmitterState(Inactive Response=::NewEnumerator1, Life Cycle Mode=::NewEnumerator1, Loop Duration=1, UseLoopDelay=false)
   /SetVariables(Emitter.SpawnNum:NiagaraFloat=&MakeCustomFloatFromBool(
     Bool=&SetBoolByFloatComparison(A=$User.Type_Transformation, B=16)
     True Float=10
     False Float=&MakeCustomFloatFromBool(Bool=&SetBoolByFloatComparison(A=$User.Type_Transformation, B=8), True Float=20, False Float=25)
    )) {isSetParameters=true, module=SetVariables_#1}
   /SetVariables(Emitter.FESpawnNum:NiagaraFloat=&MakeCustomFloatFromBool(
     Bool=&SetBoolByFloatComparison(A=$User.Type_Transformation, B=16)
     True Float=10
     False Float=&MakeCustomFloatFromBool(Bool=&SetBoolByFloatComparison(A=$User.Type_Transformation, B=8), True Float=20, False Float=25)
    )) {isSetParameters=true}
   &SpawnRate(
    SpawnRate=&MakeCustomFloatFromBool(
     Bool=$User.bisFrontEnd
     True Float=$Emitter.FESpawnNum
     False Float=&MakeCustomFloatFromBool(
      Bool=$User.bisFullyDeployed
      True Float=&Lerp_Float(A=0, B=$Emitter.SpawnNum, Alpha=$System.FN_Integrator001.Value)
      False Float=0
     )
    )
    Use Spawn Probability=false
   )
  ]
  particleSpawnScript [
   &FN_SetSpawnGroup(Number of Spawn Groups=2)
   &InitializeParticle(
    Color=&LerpLinearColors(
     StartColor=<rgba 0 0 0 50>
     EndColor=<rgba 0 0 0 50>
     LerpFactor=&MakeCustomFloatFromBool(
      Bool=&SetBoolByFloatComparison(A=&RandomRangeFloat(Minimum=0, Maximum=1), B=0.5)
      True Float=1
      False Float=0
     )
    )
    Lifetime Max=0.6
    Lifetime Min=0.3
    Lifetime Mode=::NewEnumerator1
    Mesh Scale Min=<0.5 0.5 0.5>
    Mesh Scale Mode=::NewEnumerator2
    Sprite Rotation Mode=::NewEnumerator2
    UsePositionOffset=false
    Write Material Random=true
   )
   &SkeletalMeshLocation(
    Bone / Socket Index=$Particles.SpawnGroup
    Bone Sampling Mode=::NewEnumerator4
    Inherit Velocity=false
    MaskSpawnGroup=false
    Sampled Position Offset Coordinate Space:ENiagaraMeshTransforms=::NewEnumerator2
    Skeletal Mesh=$User.New Skeletal Mesh
   )
   &BoxLocation(
    Box Size=&FN_FrontEndVector3Modifier(Gameplay=<100 3 3>, Front End=<3 3 100>)
    MaskSpawnGroup=false
    Module.Override Seed=false
    Rotate Box=false
    Surface Only=false
   )
   &AddVelocity(Velocity=&FN_FrontEndVector3Modifier(
     Gameplay=&RandomRangeVector(Minimum=<0 100 0>, Maximum=<0 300 0>)
     Front End=&RandomRangeVector(Minimum=<-50 -300 0>, Maximum=<-50 -400 0>)
    ))
   &CameraOffset(Camera Offset Amount=&RandomRangeFloat(Minimum=0, Maximum=15))
   &InitialMeshOrientation(
    Rotation=&RandomRangeVector(Minimum=<0 0 0>, Maximum=<1 1 1>)
    Use Orientation Vector=true
    Use Rotation Vector=true
   )
  ]
  particleUpdateScript [
   &ParticleState()
   &Drag(
    Drag=&FN_FrontEndFloatModifier(Gameplay=5, Front End=0)
    Use Linear Drag=true
    Use Rotational Drag=false
   )
   &InheritVelocity(
    Module.Limit Speed=true
    Velocity Limit:NiagaraFloat=&FN_FrontEndFloatModifier(Gameplay=50, Front End=0)
    Velocity Scale=<-1 -1 -1>
    Velocity Source=&TransformVector(Vector=$Engine.Owner.Velocity, Source Space=::World, Destination Space=::Local, Apply Scale=true)
   )
   &SolveForcesAndVelocity(Clamp Velocity=false, Limit Acceleration=false)
   &ScaleColor(Scale Alpha=50, ScaleA=true, ScaleRGB=true) {module=ScaleColor001}
   &ScaleMeshSize(Scale Factor=&VectorFromFloat(Value=&FloatFromCurve(FloatCurve=DataInterface() {propertyValues {Curve {keys [<at 0 0>,<at 0.2 1>,<at 0.6 1>,<at 1 0.2>]}}})))
   &UpdateMeshOrientation(
    Delta Time:NiagaraFloat=$Particles.NormalizedAge
    Rotation Rate=&RandomRangeFloat(Minimum=-0.1, Maximum=0.1)
    Rotation Vector=<1 1 1>
   )
  ]
  rendererClasses [&NiagaraMeshRendererProperties]
  renderers [{rendererClass=&NiagaraMeshRendererProperties, rendererIndex=0}]
 }
 Crystal_Piece_Gold {
  emitterUpdateScript [
   &EmitterState(Inactive Response=::NewEnumerator1, Life Cycle Mode=::NewEnumerator1, Loop Duration=1, UseLoopDelay=false)
   /SetVariables(Emitter.SpawnNum:NiagaraFloat=&MakeCustomFloatFromBool(
     Bool=&SetBoolByFloatComparison(A=$User.Type_Transformation, B=16)
     True Float=15
     False Float=&MakeCustomFloatFromBool(Bool=&SetBoolByFloatComparison(A=$User.Type_Transformation, B=8), True Float=5, False Float=0)
    )) {isSetParameters=true, module=SetVariables_#1}
   /SetVariables(Emitter.FESpawnNum:NiagaraFloat=&MakeCustomFloatFromBool(
     Bool=&SetBoolByFloatComparison(A=$User.Type_Transformation, B=16)
     True Float=15
     False Float=&MakeCustomFloatFromBool(Bool=&SetBoolByFloatComparison(A=$User.Type_Transformation, B=8), True Float=5, False Float=0)
    )) {isSetParameters=true}
   &SpawnRate(
    SpawnRate=&MakeCustomFloatFromBool(
     Bool=$User.bisFrontEnd
     True Float=$Emitter.FESpawnNum
     False Float=&MakeCustomFloatFromBool(
      Bool=$User.bisFullyDeployed
      True Float=&Lerp_Float(A=0, B=$Emitter.SpawnNum, Alpha=$System.FN_Integrator001.Value)
      False Float=0
     )
    )
    Use Spawn Probability=false
   )
  ]
  particleSpawnScript [
   &FN_SetSpawnGroup(Number of Spawn Groups=2)
   &InitializeParticle(
    Lifetime Max=0.6
    Lifetime Min=0.3
    Lifetime Mode=::NewEnumerator1
    Mesh Scale Min=<0.5 0.5 0.5>
    Mesh Scale Mode=::NewEnumerator2
    Sprite Rotation Mode=::NewEnumerator2
    UsePositionOffset=false
    Write Material Random=true
   )
   &SkeletalMeshLocation(
    Bone / Socket Index=$Particles.SpawnGroup
    Bone Sampling Mode=::NewEnumerator4
    Inherit Velocity=false
    MaskSpawnGroup=false
    Sampled Position Offset Coordinate Space:ENiagaraMeshTransforms=::NewEnumerator2
    Skeletal Mesh=$User.New Skeletal Mesh
   )
   &BoxLocation(
    Box Size=&FN_FrontEndVector3Modifier(Gameplay=<100 3 3>, Front End=<3 3 100>)
    MaskSpawnGroup=false
    Module.Override Seed=false
    Rotate Box=false
    Surface Only=false
   )
   &AddVelocity(Velocity=&FN_FrontEndVector3Modifier(
     Gameplay=&RandomRangeVector(Minimum=<0 100 0>, Maximum=<0 300 0>)
     Front End=&RandomRangeVector(Minimum=<-50 -300 0>, Maximum=<-50 -400 0>)
    ))
   &CameraOffset(Camera Offset Amount=&RandomRangeFloat(Minimum=0, Maximum=15))
   &InitialMeshOrientation(
    Rotation=&RandomRangeVector(Minimum=<0 0 0>, Maximum=<1 1 1>)
    Use Orientation Vector=true
    Use Rotation Vector=true
   )
  ]
  particleUpdateScript [
   &ParticleState()
   &Drag(
    Drag=&FN_FrontEndFloatModifier(Gameplay=5, Front End=0)
    Use Linear Drag=true
    Use Rotational Drag=false
   )
   &InheritVelocity(
    Module.Limit Speed=true
    Velocity Limit:NiagaraFloat=&FN_FrontEndFloatModifier(Gameplay=50, Front End=0)
    Velocity Scale=<-1 -1 -1>
    Velocity Source=&TransformVector(Vector=$Engine.Owner.Velocity, Source Space=::World, Destination Space=::Local, Apply Scale=true)
   )
   &SolveForcesAndVelocity(Clamp Velocity=false, Limit Acceleration=false)
   &ScaleColor(
    Scale Alpha=&FloatFromCurve(FloatCurve=DataInterface() {propertyValues {Curve {keys [<at 0.2 1>,<at 1 1>]}}})
    Scale RGB=&MakeVectorFromLinearColorRGB(Color=&RandomRangeLinearColor(Minimum=<rgba 5 1.8576 0.2865 50>, Maximum=<rgba 1 0.241 0.0512 50>))
    ScaleA=true
    ScaleRGB=true
   ) {module=ScaleColor001}
   &ScaleMeshSize(Scale Factor=&VectorFromFloat(Value=&FloatFromCurve(FloatCurve=DataInterface() {propertyValues {Curve {keys [<at 0 0>,<at 0.2 1>,<at 0.6 1>,<at 1 0.2>]}}})))
   &UpdateMeshOrientation(
    Delta Time:NiagaraFloat=$Particles.NormalizedAge
    Rotation Rate=&RandomRangeFloat(Minimum=-0.1, Maximum=0.1)
    Rotation Vector=<1 1 1>
   )
  ]
  rendererClasses [&NiagaraMeshRendererProperties]
  renderers [{rendererClass=&NiagaraMeshRendererProperties, rendererIndex=0}]
 }
 Crystal_Trail {
  emitterSpawnScript [&FN_RibbonIDOffsetUtility()]
  emitterUpdateScript [
   /SetVariables(Emitter.Only Crystal Trail=1) {isSetParameters=true}
   &EmitterState()
   &FN_SpawnAnimTrail(Enable:NiagaraBool=$User.bisFullyDeployed, Number of Anim Trails=2)
  ]
  particleSpawnScript [
   &FN_InitializeAnimTrailBoneIndecies()
   &FN_InitializeAnimTrailBoneIndecies(AnimTrail Selection=1, First Bone or Socket Index=2, Second Bone or Socket Index=3) {module=FN_InitializeAnimTrailBoneIndecies001}
   /SetVariables(Particles.AlphaFadeIn:NiagaraFloat=/Niagara/DynamicInputs/Math/Min_Float(A:NiagaraFloat=&Divide_Float(A=$Emitter.Age, B=1), B:NiagaraFloat=1)) {isSetParameters=true, module=SetVariables_#2}
   &FN_InitializeAnimTrail(
    AnimTrailWidth=0.4
    Color=&MakeLinearColorFromVectorAndFloat("Vector (RGB)"=<2 2 2>, "Float (Alpha)"=&Multiply_Float(A=1, B=$Particles.AlphaFadeIn))
    Lifetime=0.2
    Write Material Random=false
   )
   &AddVelocity(Velocity=&Lerp_Vector(
     A=<0 0 0>
     B=<-200 -1500 0>
     Alpha=&MakeCustomFloatFromBool(Bool=$User.bisFrontEnd, True Float=1, False Float=0)
    ))
   &DynamicMaterialParameters(
    Index 0 Param 4=1
    Index 1 Param 1=1
    Index 1 Param 4=1
    Module.Param0WriteEnabled=true
    Module.Param0WriteEnabled1=true
    Module.Param1WriteEnabled=true
    Module.Param1WriteEnabled1=true
    Module.Param2WriteEnabled=true
    Module.Param2WriteEnabled1=true
    Module.Param3WriteEnabled=true
    Module.Param3WriteEnabled1=true
    Write Parameter Index 1=true
   )
  ]
  particleUpdateScript [
   &SolveForcesAndVelocity(Clamp Velocity=false, Limit Acceleration=false)
   /Game/Effects/Niagara/Modules/AnimTrail/FN_UpdateAnimTrail/FN_UpdateAnimTrail(New float=1)
   &ScaleColor(
    Scale Alpha=&FloatFromCurve(
     FloatCurve=DataInterface() {propertyValues {Curve {keys [<at 0 0 RCIM_Cubic>, {arriveTangent=2.0248, interpMode=RCIM_Cubic, leaveTangent=2.0248, tangentMode=RCTM_User, time=1, value=1}]}}}
     CurveIndex=&MakeCustomFloatFromBool(Bool=$User.bisFrontEnd, True Float=1, False Float=$System.FN_Integrator001.Value)
    )
    Scale RGB=&FN_FrontEndVector3Modifier(
     Gameplay=&MakeCustomVectorFromBool(
      Bool=&SetBoolByFloatComparison(A=$User.Type_Transformation, B=8)
      True Vector=&MakeVectorFromLinearColorRGB(Color=&ColorFromCurve(ColorCurve=DataInterface() {propertyValues {
          AlphaCurve {keys [<at 0 1>]}
          BlueCurve {keys [<at 0 0.0208>,<at 1 0>]}
          GreenCurve {keys [<at 0 0.592>,<at 1 0.1034>]}
          RedCurve {keys [<at 0 1>,<at 1 1>]}
         }}))
      False Vector=&MakeVectorFromLinearColorRGB(Color=&ColorFromCurve(ColorCurve=DataInterface() {propertyValues {
          AlphaCurve {keys [<at 0 1>]}
          BlueCurve {keys [<at 0 0>,<at 1 0>]}
          GreenCurve {keys [<at 0 0>,<at 1 0>]}
          RedCurve {keys [<at 0 0>,<at 1 0>]}
         }}))
     )
     Front End=&MakeCustomVectorFromBool(
      Bool=&SetBoolByFloatComparison(A=$User.Type_Transformation, B=8)
      True Vector=&MakeVectorFromLinearColorRGB(Color=&ColorFromCurve(ColorCurve=DataInterface() {propertyValues {
          AlphaCurve {keys [<at 0 1>]}
          BlueCurve {keys [<at 0 0.0521>,<at 0.4469 0.0833>]}
          GreenCurve {keys [<at 0 0.526>,<at 0.4469 0.3125>]}
          RedCurve {keys [<at 0 1>,<at 0.4469 1>]}
         }}))
      False Vector=&MakeVectorFromLinearColorRGB(Color=&ColorFromCurve(ColorCurve=DataInterface() {propertyValues {
          AlphaCurve {keys [<at 0 1>]}
          BlueCurve {keys [<at 0 0.5052>,<at 0.4469 0>]}
          GreenCurve {keys [<at 0 0.5052>,<at 0.4469 0>]}
          RedCurve {keys [<at 0 0.5052>,<at 0.4469 0>]}
         }}))
     )
    )
    ScaleA=true
    ScaleRGB=true
   )
   &ScaleColor(
    Scale Alpha=&FN_FrontEndFloatModifier(
     Gameplay=&FloatFromCurve(FloatCurve=DataInterface() {propertyValues {Curve {keys [<at 0.5 1>,<at 1 0>]}}})
     Front End=&FloatFromCurve(FloatCurve=DataInterface() {propertyValues {Curve {keys [<at 0.5 1>,<at 1 0>]}}})
    )
    Scale RGB=&VectorFromFloat(Value=&FloatFromCurve(FloatCurve=DataInterface() {propertyValues {Curve {keys [<at 0 5>,<at 0.2 2>,<at 1 1>]}}}))
    ScaleA=true
    ScaleRGB=true
   ) {module=ScaleColor001}
   &ScaleColor(
    Scale Alpha=&MakeCustomFloatFromBool(
     Bool=&SetBoolByFloatComparison(A=$User.Type_Transformation, B=16)
     True Float=1
     False Float=&MakeCustomFloatFromBool(Bool=&SetBoolByFloatComparison(A=$User.Type_Transformation, B=8), True Float=1, False Float=3)
    )
    Scale RGB=&VectorFromFloat(Value=&MakeCustomFloatFromBool(
      Bool=&SetBoolByFloatComparison(A=$User.Type_Transformation, B=16)
      True Float=1
      False Float=&MakeCustomFloatFromBool(
       Bool=&SetBoolByFloatComparison(A=$User.Type_Transformation, B=8)
       True Float=0.75
       False Float=&FN_FrontEndFloatModifier(
        Gameplay=&FloatFromCurve(FloatCurve=DataInterface() {propertyValues {Curve {keys [<at 0 0.05>,<at 0.3 0.01>]}}}, Scale Curve=0)
        Front End=&FloatFromCurve(FloatCurve=DataInterface() {propertyValues {Curve {keys [<at 0 0.05>,<at 0.15 0.01>]}}}, Scale Curve=0)
       )
      )
     ))
    ScaleA=true
    ScaleRGB=true
   ) {module=ScaleColor002}
   /SetVariables(Particles.RibbonU0Override=$Particles.NormalizedAge) {isSetParameters=true, module=SetVariables_#3}
   &self:SocketDirection(
    First Bone:NiagaraInt32=&SelectIntFromArray(Direct Array Index=$Particles.SpawnGroup, Int Selection Array=DataInterface() {propertyValues {IntData [0,2]}})
    Second Bone:NiagaraInt32=&SelectIntFromArray(Direct Array Index=$Particles.SpawnGroup, Int Selection Array=DataInterface() {propertyValues {IntData [1,3]}})
    Skeletal Mesh:NiagaraDataInterfaceSkeletalMesh=$User.SkeletalMesh
   ) {module=SocketDirection}
   /SetVariables(Particles.OffsetPosition:Vector3f=&Add_Vector(
     A=Linked() {linkedVariable {name=Particles.Position, type=NiagaraPosition}}
     B=&Multiply_Vector(A=$Particles.Direction, B=&VectorFromFloat(Value=-40))
    )) {isSetParameters=true}
  ]
  rendererClasses [&NiagaraRibbonRendererProperties]
  renderers [{rendererClass=&NiagaraRibbonRendererProperties}]
 }
 Sand_Trail_Gold {
  emitterSpawnScript [&FN_RibbonIDOffsetUtility()]
  emitterUpdateScript [
   &EmitterState()
   &FN_SpawnAnimTrail(Enable:NiagaraBool=$User.bisFullyDeployed, Number of Anim Trails=2)
  ]
  particleSpawnScript [
   &FN_InitializeAnimTrailBoneIndecies()
   &FN_InitializeAnimTrailBoneIndecies(AnimTrail Selection=1, First Bone or Socket Index=2, Second Bone or Socket Index=3) {module=FN_InitializeAnimTrailBoneIndecies001}
   /SetVariables(Particles.AlphaFadeIn:NiagaraFloat=&Min_Float(A:NiagaraFloat=&Divide_Float(A=$Emitter.Age, B=1), B:NiagaraFloat=1)) {isSetParameters=true, module=SetVariables_#2}
   &FN_InitializeAnimTrail(
    Color=&MakeLinearColorFromVectorAndFloat("Vector (RGB)"=<100 100 100>, "Float (Alpha)"=&Multiply_Float(A=1, B=$Particles.AlphaFadeIn))
    Lifetime=0.2
    Write Material Random=false
   )
   &AddVelocity(Velocity=&Lerp_Vector(
     A=<0 0 0>
     B=<-200 -1500 0>
     Alpha=&MakeCustomFloatFromBool(Bool=$User.bisFrontEnd, True Float=1, False Float=0)
    ))
   &DynamicMaterialParameters(
    Index 0 Param 3=0.5
    Index 0 Param 4=1
    Index 1 Param 1=1
    Index 1 Param 4=1
    Module.Param0WriteEnabled=true
    Module.Param0WriteEnabled1=true
    Module.Param1WriteEnabled=true
    Module.Param1WriteEnabled1=true
    Module.Param2WriteEnabled=true
    Module.Param2WriteEnabled1=true
    Module.Param3WriteEnabled=true
    Module.Param3WriteEnabled1=true
    Write Parameter Index 1=true
   )
  ]
  particleUpdateScript [
   &SolveForcesAndVelocity(Clamp Velocity=false, Limit Acceleration=false)
   &FN_UpdateAnimTrail(New float=1)
   &ScaleColor(
    Scale Alpha=&FloatFromCurve(
     FloatCurve=DataInterface() {propertyValues {Curve {keys [<at 0 0 RCIM_Cubic>, {arriveTangent=2.0248, interpMode=RCIM_Cubic, leaveTangent=2.0248, tangentMode=RCTM_User, time=1, value=1}]}}}
     CurveIndex=&MakeCustomFloatFromBool(Bool=$User.bisFrontEnd, True Float=1, False Float=$System.FN_Integrator001.Value)
    )
    Scale RGB=&MakeVectorFromLinearColorRGB(Color=&ColorFromCurve(ColorCurve=DataInterface() {propertyValues {
        AlphaCurve {keys [<at 0 1>]}
        BlueCurve {keys [<at 0 0.2083>,<at 0.349 0.2083>,<at 1 0>]}
        GreenCurve {keys [<at 0 5.9201>,<at 0.349 4.2882>,<at 1 0.1034>]}
        RedCurve {keys [<at 0 10>,<at 0.349 10>,<at 1 1>]}
       }}))
    ScaleA=true
    ScaleRGB=true
   )
   &ScaleColor(
    Scale Alpha=&FN_FrontEndFloatModifier(
     Gameplay=&FloatFromCurve(FloatCurve=DataInterface() {propertyValues {Curve {keys [<at 0.5 1>,<at 1 0>]}}})
     Front End=&FloatFromCurve(FloatCurve=DataInterface() {propertyValues {Curve {keys [<at 0.5 1>,<at 1 0>]}}}, Scale Curve=0.2)
    )
    ScaleA=true
    ScaleRGB=true
   ) {module=ScaleColor001}
   &ScaleColor(
    Scale Alpha=&MakeCustomFloatFromBool(Bool=&SetBoolByFloatComparison(A=$User.Type_Transformation, B=16), True Float=1, False Float=0)
    ScaleA=true
    ScaleRGB=true
   ) {module=ScaleColor002}
   /SetVariables(Particles.RibbonU0Override=$Particles.NormalizedAge) {isSetParameters=true, module=SetVariables_#3}
   &self:SocketDirection(
    First Bone:NiagaraInt32=&SelectIntFromArray(Direct Array Index=$Particles.SpawnGroup, Int Selection Array=DataInterface() {propertyValues {IntData [0,2]}})
    Second Bone:NiagaraInt32=&SelectIntFromArray(Direct Array Index=$Particles.SpawnGroup, Int Selection Array=DataInterface() {propertyValues {IntData [1,3]}})
    Skeletal Mesh:NiagaraDataInterfaceSkeletalMesh=$User.SkeletalMesh
   ) {module=SocketDirection}
   /SetVariables(Particles.OffsetPosition:Vector3f=&Add_Vector(
     A=Linked() {linkedVariable {name=Particles.Position, type=NiagaraPosition}}
     B=&Multiply_Vector(A=$Particles.Direction, B=&VectorFromFloat(Value=-60))
    )) {isSetParameters=true, module=SetVariables_#4}
  ]
  rendererClasses [&NiagaraRibbonRendererProperties]
  renderers [{rendererClass=&NiagaraRibbonRendererProperties}]
 }
 Sand_Trail_Orange {
  emitterSpawnScript [&FN_RibbonIDOffsetUtility()]
  emitterUpdateScript [
   &EmitterState()
   &FN_SpawnAnimTrail(Enable:NiagaraBool=$User.bisFullyDeployed, Number of Anim Trails=2)
  ]
  particleSpawnScript [
   &FN_InitializeAnimTrailBoneIndecies()
   &FN_InitializeAnimTrailBoneIndecies(AnimTrail Selection=1, First Bone or Socket Index=2, Second Bone or Socket Index=3) {module=FN_InitializeAnimTrailBoneIndecies001}
   /SetVariables(Particles.AlphaFadeIn:NiagaraFloat=&Min_Float(A:NiagaraFloat=&Divide_Float(A=$Emitter.Age, B=1), B:NiagaraFloat=1)) {isSetParameters=true, module=SetVariables_#2}
   &FN_InitializeAnimTrail(
    Color=&MakeLinearColorFromVectorAndFloat("Vector (RGB)"=<1.5 1.5 1.5>, "Float (Alpha)"=&Multiply_Float(A=0.5, B=$Particles.AlphaFadeIn))
    Lifetime=0.2
    Write Material Random=false
   )
   &AddVelocity(Velocity=&Lerp_Vector(
     A=<0 0 0>
     B=<-200 -1500 0>
     Alpha=&MakeCustomFloatFromBool(Bool=$User.bisFrontEnd, True Float=1, False Float=0)
    ))
   &DynamicMaterialParameters(
    Index 0 Param 3=0.5
    Index 0 Param 4=1
    Index 1 Param 1=1
    Index 1 Param 4=1
    Module.Param0WriteEnabled=true
    Module.Param0WriteEnabled1=true
    Module.Param1WriteEnabled=true
    Module.Param1WriteEnabled1=true
    Module.Param2WriteEnabled=true
    Module.Param2WriteEnabled1=true
    Module.Param3WriteEnabled=true
    Module.Param3WriteEnabled1=true
    Write Parameter Index 1=true
   )
  ]
  particleUpdateScript [
   &SolveForcesAndVelocity(Clamp Velocity=false, Limit Acceleration=false)
   &FN_UpdateAnimTrail(New float=1)
   &ScaleColor(
    Scale Alpha=&FloatFromCurve(
     FloatCurve=DataInterface() {propertyValues {Curve {keys [<at 0 0 RCIM_Cubic>, {arriveTangent=2.0248, interpMode=RCIM_Cubic, leaveTangent=2.0248, tangentMode=RCTM_User, time=1, value=1}]}}}
     CurveIndex=&MakeCustomFloatFromBool(Bool=$User.bisFrontEnd, True Float=1, False Float=$System.FN_Integrator001.Value)
    )
    Scale RGB=&MakeVectorFromLinearColorRGB(Color=&ColorFromCurve(ColorCurve=DataInterface() {propertyValues {
        AlphaCurve {keys [<at 0 1>]}
        BlueCurve {keys [<at 0 0.0208>,<at 1 0>]}
        GreenCurve {keys [<at 0 0.4288>,<at 1 0.1034>]}
        RedCurve {keys [<at 0 1>,<at 1 1>]}
       }}))
    ScaleA=true
    ScaleRGB=true
   )
   &ScaleColor(
    Scale Alpha=&FN_FrontEndFloatModifier(
     Gameplay=&FloatFromCurve(FloatCurve=DataInterface() {propertyValues {Curve {keys [<at 0.5 1>,<at 1 0>]}}})
     Front End=&FloatFromCurve(FloatCurve=DataInterface() {propertyValues {Curve {keys [<at 0.5 1>,<at 1 0>]}}}, Scale Curve=0.2)
    )
    ScaleA=true
    ScaleRGB=true
   ) {module=ScaleColor001}
   &ScaleColor(
    Scale Alpha=&MakeCustomFloatFromBool(Bool=&SetBoolByFloatComparison(A=$User.Type_Transformation, B=16), True Float=1, False Float=0)
    ScaleA=true
    ScaleRGB=true
   ) {module=ScaleColor002}
   /SetVariables(Particles.RibbonU0Override=$Particles.NormalizedAge) {isSetParameters=true, module=SetVariables_#3}
   &self:SocketDirection(
    First Bone:NiagaraInt32=&SelectIntFromArray(Direct Array Index=$Particles.SpawnGroup, Int Selection Array=DataInterface() {propertyValues {IntData [0,2]}})
    Second Bone:NiagaraInt32=&SelectIntFromArray(Direct Array Index=$Particles.SpawnGroup, Int Selection Array=DataInterface() {propertyValues {IntData [1,3]}})
    Skeletal Mesh:NiagaraDataInterfaceSkeletalMesh=$User.SkeletalMesh
   ) {module=SocketDirection}
   /SetVariables(Particles.OffsetPosition:Vector3f=&Add_Vector(
     A=Linked() {linkedVariable {name=Particles.Position, type=NiagaraPosition}}
     B=&Multiply_Vector(A=$Particles.Direction, B=&VectorFromFloat(Value=5))
    )) {isSetParameters=true, module=SetVariables_#4}
  ]
  rendererClasses [&NiagaraRibbonRendererProperties]
  renderers [{rendererClass=&NiagaraRibbonRendererProperties}]
 }
 Smoke_Trail {
  emitterSpawnScript [&FN_RibbonIDOffsetUtility()]
  emitterUpdateScript [
   &EmitterState()
   &FN_SpawnAnimTrail(Enable:NiagaraBool=$User.bisFullyDeployed, Number of Anim Trails=2)
  ]
  particleSpawnScript [
   &FN_InitializeAnimTrailBoneIndecies()
   &FN_InitializeAnimTrailBoneIndecies(AnimTrail Selection=1, First Bone or Socket Index=2, Second Bone or Socket Index=3) {module=FN_InitializeAnimTrailBoneIndecies001}
   /SetVariables(Particles.AlphaFadeIn:NiagaraFloat=&Min_Float(A:NiagaraFloat=&Divide_Float(A=$Emitter.Age, B=1), B:NiagaraFloat=1)) {isSetParameters=true, module=SetVariables_#2}
   &FN_InitializeAnimTrail(
    AnimTrailWidth=0.75
    Color=&MakeLinearColorFromVectorAndFloat("Vector (RGB)"=<1 1 1>, "Float (Alpha)"=&Multiply_Float(A=1, B=$Particles.AlphaFadeIn))
    Lifetime=0.2
    Write Material Random=false
   )
   &AddVelocity(Velocity=&Lerp_Vector(
     A=<0 0 0>
     B=<-200 -1500 0>
     Alpha=&MakeCustomFloatFromBool(Bool=$User.bisFrontEnd, True Float=1, False Float=0)
    ))
  ]
  particleUpdateScript [
   &SolveForcesAndVelocity(Clamp Velocity=false, Limit Acceleration=false)
   &FN_UpdateAnimTrail(New float=1)
   &ScaleColor(
    Scale Alpha=&FloatFromCurve(
     FloatCurve=DataInterface() {propertyValues {Curve {keys [<at 0 0 RCIM_Cubic>, {arriveTangent=2.0248, interpMode=RCIM_Cubic, leaveTangent=2.0248, tangentMode=RCTM_User, time=1, value=1}]}}}
     CurveIndex=&MakeCustomFloatFromBool(Bool=$User.bisFrontEnd, True Float=1, False Float=$System.FN_Integrator001.Value)
    )
    Scale RGB=&MakeCustomVectorFromBool(
     Bool=&SetBoolByFloatComparison(A=$User.Type_Transformation, B=16)
     True Vector=&MakeVectorFromLinearColorRGB(Color=&ColorFromCurve(ColorCurve=DataInterface() {propertyValues {
         AlphaCurve {keys [<at 0 1>]}
         BlueCurve {keys [<at 0 1.0474>,<at 0.234 0.1042>,<at 1 0>]}
         GreenCurve {keys [<at 0 10.5237>,<at 0.234 2.1441>,<at 1 0.3333>]}
         RedCurve {keys [<at 0 20>,<at 0.234 5>,<at 1 1>]}
        }}))
     False Vector=&FN_FrontEndVector3Modifier(
      Gameplay=&MakeVectorFromLinearColorRGB(Color=&ColorFromCurve(ColorCurve=DataInterface() {propertyValues {
          AlphaCurve {keys [<at 0 1>]}
          BlueCurve {keys [<at 0 0.1458>,<at 1 0.0208>]}
          GreenCurve {keys [<at 0 0.7153>,<at 1 0.4288>]}
          RedCurve {keys [<at 0 1>,<at 1 1>]}
         }}))
      Front End=&MakeCustomVectorFromBool(
       Bool=&SetBoolByFloatComparison(A=$User.Type_Transformation, B=8)
       True Vector=&MakeVectorFromLinearColorRGB(Color=&ColorFromCurve(ColorCurve=DataInterface() {propertyValues {
           AlphaCurve {keys [<at 0 1>]}
           BlueCurve {keys [<at 0 0.3542>,<at 0.406 0.099>]}
           GreenCurve {keys [<at 0 1.0399>,<at 0.406 0.4744>]}
           RedCurve {keys [<at 0 2>,<at 0.406 1>]}
          }}))
       False Vector=&MakeVectorFromLinearColorRGB(Color=&ColorFromCurve(ColorCurve=DataInterface() {propertyValues {
           AlphaCurve {keys [<at 0 1>]}
           BlueCurve {keys [<at 0 0.4688>,<at 0.4135 0>]}
           GreenCurve {keys [<at 0 0.4688>,<at 0.4135 0>]}
           RedCurve {keys [<at 0 0.4688>,<at 0.4135 0>]}
          }}))
      )
     )
    )
    ScaleA=true
    ScaleRGB=true
   )
   &ScaleColor(
    Scale Alpha=&FN_FrontEndFloatModifier(
     Gameplay=&FloatFromCurve(FloatCurve=DataInterface() {propertyValues {Curve {keys [<at 0.5 1>,<at 1 0>]}}})
     Front End=&FloatFromCurve(FloatCurve=DataInterface() {propertyValues {Curve {keys [<at 0.5 1>,<at 1 0>]}}})
    )
    Scale RGB=&VectorFromFloat(Value=&MakeCustomFloatFromBool(
      Bool=&SetBoolByFloatComparison(A=$User.Type_Transformation, B=16)
      True Float=&FloatFromCurve(FloatCurve=DataInterface() {propertyValues {Curve {keys [<at 0 10>, {arriveTangent=-4.7105, interpMode=RCIM_Cubic, leaveTangent=-4.7105, tangentMode=RCTM_User, time=0.2, value=3}, <at 1 1>]}}})
      False Float=&FN_FrontEndFloatModifier(
       Gameplay=&FloatFromCurve(FloatCurve=DataInterface() {propertyValues {Curve {keys [<at 0 1>, {arriveTangent=-0.4564, interpMode=RCIM_Cubic, leaveTangent=-0.4564, tangentMode=RCTM_User, time=0.3, value=0.1}, <at 1 0>]}}})
       Front End=&FloatFromCurve(FloatCurve=DataInterface() {propertyValues {Curve {keys [
            <at 0 0.2 RCIM_Cubic>
            {arriveTangent=-0.3333, interpMode=RCIM_Cubic, leaveTangent=-0.3333, time=0.15, value=0.3}
            {arriveTangent=0.0358, interpMode=RCIM_Cubic, leaveTangent=0.0358, tangentMode=RCTM_User, time=0.3, value=0.1}
            <at 1 0>
           ]}}})
      )
     ))
    ScaleA=true
    ScaleRGB=true
   ) {module=ScaleColor001}
   &ScaleColor(
    Scale Alpha=&MakeCustomFloatFromBool(
     Bool=&SetBoolByFloatComparison(A=$User.Type_Transformation, B=16)
     True Float=1
     False Float=&FN_FrontEndFloatModifier(
      Gameplay=&MakeCustomFloatFromBool(
       Bool=&SetBoolByFloatComparison(A=$User.Type_Transformation, B=8)
       True Float=0.8
       False Float=&FN_FrontEndFloatModifier(Gameplay=0.4, Front End=0.5)
      )
      Front End=0.5
     )
    )
    Scale RGB=&VectorFromFloat(Value=&MakeCustomFloatFromBool(
      Bool=&SetBoolByFloatComparison(A=$User.Type_Transformation, B=16)
      True Float=1
      False Float=&MakeCustomFloatFromBool(
       Bool=&SetBoolByFloatComparison(A=$User.Type_Transformation, B=8)
       True Float=&FN_FrontEndFloatModifier(Gameplay=0.2, Front End=0.5)
       False Float=&FN_FrontEndFloatModifier(Gameplay=0, Front End=0)
      )
     ))
    ScaleA=true
    ScaleRGB=true
   ) {module=ScaleColor002}
   &DynamicMaterialParameters(
    Index 0 Param 1=&MakeCustomFloatFromBool(
     Bool=&SetBoolByFloatComparison(A=$User.Type_Transformation, B=16)
     True Float=1
     False Float=&MakeCustomFloatFromBool(Bool=&SetBoolByFloatComparison(A=$User.Type_Transformation, B=8), True Float=0.01, False Float=0)
    )
    Index 0 Param 2=&MakeCustomFloatFromBool(
     Bool=&SetBoolByFloatComparison(A=$User.Type_Transformation, B=16)
     True Float=0
     False Float=&MakeCustomFloatFromBool(Bool=&SetBoolByFloatComparison(A=$User.Type_Transformation, B=8), True Float=0.5, False Float=0)
    )
    Index 0 Param 3=1
    Index 0 Param 4=1
    Module.Param0WriteEnabled=true
    Module.Param1WriteEnabled=true
    Module.Param2WriteEnabled=true
    Module.Param3WriteEnabled=true
   )
   /SetVariables(Particles.RibbonU0Override=$Particles.NormalizedAge) {isSetParameters=true, module=SetVariables_#3}
   &self:SocketDirection(
    First Bone:NiagaraInt32=&SelectIntFromArray(Direct Array Index=$Particles.SpawnGroup, Int Selection Array=DataInterface() {propertyValues {IntData [0,2]}})
    Second Bone:NiagaraInt32=&SelectIntFromArray(Direct Array Index=$Particles.SpawnGroup, Int Selection Array=DataInterface() {propertyValues {IntData [1,3]}})
    Skeletal Mesh:NiagaraDataInterfaceSkeletalMesh=$User.SkeletalMesh
   ) {module=SocketDirection}
   /SetVariables(Particles.OffsetPosition:Vector3f=&Add_Vector(
     A=Linked() {linkedVariable {name=Particles.Position, type=NiagaraPosition}}
     B=&Multiply_Vector(A=$Particles.Direction, B=&VectorFromFloat(Value=-17))
    )) {isSetParameters=true, module=SetVariables_#4}
  ]
  rendererClasses [&NiagaraRibbonRendererProperties]
  renderers [{rendererClass=&NiagaraRibbonRendererProperties}]
 }
}
summary {
 systemName=NS_Glider_Majesty_Male_Trail
 emitters [
  {emitterName=Smoke_Trail, rendererClasses [&NiagaraRibbonRendererProperties]}
  {emitterName=Sand_Trail_Orange, rendererClasses [&NiagaraRibbonRendererProperties]}
  {emitterName=Crystal_Piece_Gold, rendererClasses [&NiagaraMeshRendererProperties]}
  {emitterName=Sand_Trail_Gold, rendererClasses [&NiagaraRibbonRendererProperties]}
  {emitterName=Crystal_Piece_Dark, rendererClasses [&NiagaraMeshRendererProperties]}
  {emitterName=Crystal_Trail, rendererClasses [&NiagaraRibbonRendererProperties]}
 ]
 userVariables [
  User.bisFrontEnd {type=NiagaraBool}
  User.bIsFrontEndPreview {type=NiagaraBool}
  User.bisFullyDeployed {type=NiagaraBool}
  User.ForwardDot {type=NiagaraFloat, defaultValue=0, description=""}
  User.New Skeletal Mesh {
   type=NiagaraDataInterfaceSkeletalMesh
   defaultValue {
    struct=NiagaraExt_VariableValue_DataInterface
    value {dataInterface=&self:NiagaraDataInterfaceSkeletalMesh_0}
   }
  }
  User.RibbonIDOffset {type=NiagaraInt32}
  User.SkeletalMesh {
   type=NiagaraDataInterfaceSkeletalMesh
   defaultValue {
    struct=NiagaraExt_VariableValue_DataInterface
    value {dataInterface=&self:NiagaraDataInterfaceSkeletalMesh_2}
   }
  }
  User.Type_Transformation {type=NiagaraFloat, defaultValue=1, description=""}
 ]
}
